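# Looping ambient Sound Cue of metal wronk and groan one-shots for UFO crash sites, using metal stress one-shot attenuation.
# A loop with a random delay spaces out events, then a quality-level switch selects the content.
# The first two levels randomly pick among three layered pairs (knock+spot, spot+stretch, stretch+swell), the last weighted lower.
# The other two play single spots from a smaller pool, with modulators varying pitch and volume.
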SoundCue /Game/Sounds/Apollo/Biomes/UFO_Crash_Sites/AmbientOneshots/AMB_UFO_Crashsite_Metal_Wronks_All_Looping_Spot_Cue
root=Looping_0
nodes [
 {id=Delay_0, children [QualityLevel_0], properties {delayMax=15, delayMin=5}}
 {id=Looping_0, children [Delay_0]}
 {id=Mixer_0, children [Modulator_3,Modulator_1], properties {inputVolume [0.8,1]}}
 {id=Mixer_1, children [Modulator_1,Modulator_4], properties {inputVolume [1,1]}}
 {id=Mixer_2, children [Modulator_4,Modulator_2], properties {inputVolume [1,1]}}
 {id=Modulator_0, children [Random_2], properties {pitchMax=1.1, pitchMin=0.9, volumeMax=2.5, volumeMin=2}}
 {id=Modulator_1, children [Random_1], properties {pitchMax=1.1, pitchMin=0.9, volumeMax=3.5, volumeMin=3}}
 {id=Modulator_2, children [Random_5]}
 {id=Modulator_3, children [Random_0], properties {pitchMin=0.9, volumeMax=0.8, volumeMin=0.7}}
 {id=Modulator_4, children [Random_3], properties {pitchMin=0.9, volumeMax=0.8, volumeMin=0.6}}
 {id=QualityLevel_0, children [Random_4,Random_4,Modulator_0,Modulator_0]}
 {
  id=Random_0
  children [WavePlayer_1, WavePlayer_7, WavePlayer_0, WavePlayer_4, WavePlayer_3, WavePlayer_2, WavePlayer_20, WavePlayer_21, WavePlayer_17]
  properties {weights [1,1,1,1,1,1,1,1,1]}
 }
 {
  id=Random_1
  children [
   WavePlayer_5
   WavePlayer_6
   WavePlayer_11
   WavePlayer_12
   WavePlayer_13
   WavePlayer_14
   WavePlayer_15
   WavePlayer_16
   WavePlayer_18
   WavePlayer_19
  ]
  properties {weights [1,1,1,1,1,1,1,1,1,1]}
 }
 {id=Random_2, children [WavePlayer_25, WavePlayer_26, WavePlayer_27, WavePlayer_28, WavePlayer_29], properties {weights [1,1,1,1,1]}}
 {id=Random_3, children [WavePlayer_22, WavePlayer_23, WavePlayer_24], properties {weights [1,1,1]}}
 {id=Random_4, children [Mixer_0,Mixer_1,Mixer_2], properties {bRandomizeWithoutReplacement=false, weights [1,1,0.5]}}
 {id=Random_5, children [WavePlayer_8, WavePlayer_9, WavePlayer_10, WavePlayer_30, WavePlayer_31], properties {weights [1,1,1,1,1]}}
 {id=WavePlayer_0, properties {soundWaveAssetPtr=/Game/Sounds/Apollo/Biomes/Generic/AMB_Metal_Stress_Wronk_Groan_Knock_03}}
 {id=WavePlayer_1, properties {soundWaveAssetPtr=/Game/Sounds/Apollo/Biomes/Generic/AMB_Metal_Stress_Wronk_Groan_Knock_01}}
 {
  id=WavePlayer_10
  properties {soundWaveAssetPtr=&/AMB_UFO_Crashsite_Huge_Metal_Swell_03}
 }
 {
  id=WavePlayer_11
  properties {soundWaveAssetPtr=&/AMB_UFO_Crashsite_Metal_Wronk_Spot_01}
 }
 {
  id=WavePlayer_12
  properties {soundWaveAssetPtr=&/AMB_UFO_Crashsite_Metal_Wronk_Spot_02}
 }
 {
  id=WavePlayer_13
  properties {soundWaveAssetPtr=&/AMB_UFO_Crashsite_Metal_Wronk_Spot_03}
 }
 {
  id=WavePlayer_14
  properties {soundWaveAssetPtr=&/AMB_UFO_Crashsite_Metal_Wronk_Spot_04}
 }
 {
  id=WavePlayer_15
  properties {soundWaveAssetPtr=&/AMB_UFO_Crashsite_Metal_Wronk_Spot_05}
 }
 {
  id=WavePlayer_16
  properties {soundWaveAssetPtr=&/AMB_UFO_Crashsite_Metal_Wronk_Spot_06}
 }
 {id=WavePlayer_17, properties {soundWaveAssetPtr=/Game/Sounds/Apollo/Biomes/Generic/AMB_Metal_Stress_Wronk_Groan_Knock_09}}
 {
  id=WavePlayer_18
  properties {soundWaveAssetPtr=&/AMB_UFO_Crashsite_Metal_Wronk_Spot_07}
 }
 {
  id=WavePlayer_19
  properties {soundWaveAssetPtr=&/AMB_UFO_Crashsite_Metal_Wronk_Spot_08}
 }
 {id=WavePlayer_2, properties {soundWaveAssetPtr=/Game/Sounds/Apollo/Biomes/Generic/AMB_Metal_Stress_Wronk_Groan_Knock_06}}
 {id=WavePlayer_20, properties {soundWaveAssetPtr=/Game/Sounds/Apollo/Biomes/Generic/AMB_Metal_Stress_Wronk_Groan_Knock_07}}
 {id=WavePlayer_21, properties {soundWaveAssetPtr=/Game/Sounds/Apollo/Biomes/Generic/AMB_Metal_Stress_Wronk_Groan_Knock_08}}
 {id=WavePlayer_22, properties {soundWaveAssetPtr=/Game/Sounds/Apollo/Biomes/Generic/AMB_Metal_Stress_Wronk_Groan_Stretch_02}}
 {id=WavePlayer_23, properties {soundWaveAssetPtr=/Game/Sounds/Apollo/Biomes/Generic/AMB_Metal_Stress_Wronk_Groan_Stretch_03}}
 {id=WavePlayer_24, properties {soundWaveAssetPtr=/Game/Sounds/Apollo/Biomes/Generic/AMB_Metal_Stress_Wronk_Groan_Stretch_04}}
 {
  id=WavePlayer_25
  properties {soundWaveAssetPtr=&AMB_UFO_Crashsite_Metal_Wronk_Spot_05}
 }
 {
  id=WavePlayer_26
  properties {soundWaveAssetPtr=&AMB_UFO_Crashsite_Metal_Wronk_Spot_08}
 }
 {
  id=WavePlayer_27
  properties {soundWaveAssetPtr=&/AMB_UFO_Crashsite_Metal_Wronk_Spot_09}
 }
 {
  id=WavePlayer_28
  properties {soundWaveAssetPtr=&AMB_UFO_Crashsite_Metal_Wronk_Spot_02}
 }
 {
  id=WavePlayer_29
  properties {soundWaveAssetPtr=&/AMB_UFO_Crashsite_Metal_Wronk_Spot_10}
 }
 {id=WavePlayer_3, properties {soundWaveAssetPtr=/Game/Sounds/Apollo/Biomes/Generic/AMB_Metal_Stress_Wronk_Groan_Knock_05}}
 {
  id=WavePlayer_30
  properties {soundWaveAssetPtr=&/AMB_UFO_Crashsite_Huge_Metal_Swell_04}
 }
 {
  id=WavePlayer_31
  properties {soundWaveAssetPtr=&/AMB_UFO_Crashsite_Huge_Metal_Swell_05}
 }
 {id=WavePlayer_4, properties {soundWaveAssetPtr=/Game/Sounds/Apollo/Biomes/Generic/AMB_Metal_Stress_Wronk_Groan_Knock_04}}
 {
  id=WavePlayer_5
  properties {soundWaveAssetPtr=&AMB_UFO_Crashsite_Metal_Wronk_Spot_09}
 }
 {
  id=WavePlayer_6
  properties {soundWaveAssetPtr=&AMB_UFO_Crashsite_Metal_Wronk_Spot_10}
 }
 {id=WavePlayer_7, properties {soundWaveAssetPtr=/Game/Sounds/Apollo/Biomes/Generic/AMB_Metal_Stress_Wronk_Groan_Knock_02}}
 {
  id=WavePlayer_8
  properties {soundWaveAssetPtr=&/AMB_UFO_Crashsite_Huge_Metal_Swell_01}
 }
 {
  id=WavePlayer_9
  properties {soundWaveAssetPtr=&/AMB_UFO_Crashsite_Huge_Metal_Swell_02}
 }
]
properties {
 attenuationSettings=/Game/Sounds/Attenuation/Fort_Metal_Stress_OneShot_Att
 maxDistance=9500
 soundClassObject=&Fort_Ambient_Outdoors_Athena
 volumeMultiplier=0.501
 preEffectBusSends [{}]
}
